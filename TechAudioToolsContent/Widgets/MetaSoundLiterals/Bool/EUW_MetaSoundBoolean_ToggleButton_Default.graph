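# Editor utility widget for editing one boolean MetaSound input: a label, a toggle button and a reset-to-default button.
# Set Input Viewmodels looks up Input Name and passes the cast viewmodel to SetMetaSoundInputEditorViewModel.
# Toggle changes and resets use undo transactions on the boolean literal viewmodel, with reset writing Default Value to SourceValue.
# A visibility function hides reset at the default value, and mouse enter/leave broadcast Is Hovered as a field notify.
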EditorUtilityWidgetBlueprint /TechAudioToolsContent/Widgets/MetaSoundLiterals/Bool/EUW_MetaSoundBoolean_ToggleButton_Default
graphs [
 EventGraph {nodes [
   "On Clicked (Button)"() {is=title K2Node_ComponentBoundEvent}
   > Create Transaction!(
    Begin:Exec
    Finalize:Exec=@n1.then
    Cancel:Exec
    Context:String=GetDisplayName(Object=Self-Reference())
    Description:Text=Get Reset Undo Buffer Description!(self) -> (Text:Text) {is=title K2Node_CallFunction}
   ) -> (Pending:Exec, Complete:Exec) {is=title K2Node_MacroInstance}
   > TransactObject(Object=Get MetaSoundLiteralViewModel_Boolean() -> (MetaSoundLiteralViewModel_Boolean:Meta Sound Literal Boolean Viewmodel Object Reference))
   > n1 = Set SourceValue!(
    execute
    SourceValue:bool=Get Default Value() -> (Default Value:bool)
    self:Meta Sound Literal Boolean Viewmodel Object Reference=Get MetaSoundLiteralViewModel_Boolean()
   ) -> (then, Output_Get:bool)
   "On Begin Value Change (Bool_ToggleButton)"() {is=title K2Node_ComponentBoundEvent}
   > Create Transaction(
    Finalize=@n2.then
    Context=GetDisplayName(Object=Self-Reference())
    Description=Get Undo Buffer Description!(self) -> (Text:Text) {is=title K2Node_CallFunction}
   )
   > TransactObject(Object=Get MetaSoundLiteralViewModel_Boolean())
   n2 = "On End Value Change (Bool_ToggleButton)"() {is=title K2Node_ComponentBoundEvent}
   Event PreConstruct()
   > "SetText (Text)"(self=Get LabelTextBlock() -> (LabelTextBlock:Text Object Reference), InText=Get Label Text() -> (Label Text:Text))
   n3 = Event OnMouseEnter()
   Broadcast Field Value Changed!(execute=[@n3.then, @n4.then], self:Widget Object Reference, FieldId:Field Notification Id Structure=(FieldName="Is Hovered")) -> (then)
   n4 = Event OnMouseLeave!() -> (OutputDelegate, then, MouseEvent:"Pointer Event Structure (by ref)") {is=title K2Node_Event}
   n5 = Event Set Input Viewmodels!() -> (OutputDelegate, then, InputViewModels:Map of Names to Meta Sound Input Viewmodel Object References) {is=title K2Node_Event}
   > n6 = Cast To MetaSoundInputEditorViewModel!(
    execute
    Object=Find(
     TargetMap:Map of Names to Meta Sound Input Viewmodel Object References=@n5.InputViewModels
     Key:"Name (by ref)"=Get Input Name() -> (Input Name:Name=None)
    ) -> (Value:Meta Sound Input Viewmodel Object Reference)
   ) -> (then, CastFailed, AsMetaSound Input Editor Viewmodel:Meta Sound Input Editor Viewmodel Object Reference) {is=title K2Node_DynamicCast}
   > SetMetaSoundInputEditorViewModel!(execute, self, Viewmodel:Meta Sound Input Editor Viewmodel Object Reference=@n6) -> (then)
  ]}
 Get Reset Button Visibility {nodes [
   n1 = Get Reset Button Visibility!() -> (then, Value:bool=~) {is=title K2Node_FunctionEntry}
   n2 = Get Default Value()
   > Return Node!(
    execute
    ReturnValue:ESlateVisibility Enum=Select(
     Option 0=Visible
     Option 1=Hidden
     Index="Equal (Boolean)"(A:bool=@n1, B:bool=@n2) -> (ReturnValue:bool=~) {is=title K2Node_PromotableOperator}
    )
   )
  ]}
 Get Reset Undo Buffer Description {nodes [
   Get Reset Undo Buffer Description() {is=title K2Node_FunctionEntry}
   n1 = Get Label Text()
   n2 = TextIsEmpty(InText=@n1)
   n3 = "To Text (Name)"!(InName:Name=Get Input Name() -> (Input Name=None)) -> (ReturnValue:Text)
   > Return Node!(
    execute
    Text:Text=Format Text!(Format="LOCTEXT(\"Reset {name} to Default\")", name:Text=Select(Option 0:Text=@n1, Option 1:Text=@n3, Index=@n2) -> (ReturnValue:Text)) -> (Result)
   )
  ]}
 Get Undo Buffer Description {nodes [
   Get Undo Buffer Description() {is=title K2Node_FunctionEntry}
   n1 = Get Label Text()
   n2 = TextIsEmpty(InText=@n1)
   n3 = "To Text (Name)"(InName=Get Input Name() -> (Input Name=None))
   > Return Node(Text=Format Text!(Format="LOCTEXT(\"Set {parameter}\")", parameter:Text=Select(Option 0=@n1, Option 1=@n3, Index=@n2)) -> (Result))
  ]}
 GetInputViewModelNames {nodes [
   Get Input Viewmodel Names() {is=title K2Node_FunctionEntry}
   > Return Node!(execute, ReturnValue:Array of Names=Make Array("[0]":Name=Get Input Name() -> (Input Name=None)) -> (Array:Array of Names))
  ]}
 Is Hovered {nodes [
   Is Hovered() {is=title K2Node_FunctionEntry}
   > Return Node(ReturnValue:bool=IsHovered())
  ]}
]
variables [
 Default Value {type=boolean}
 Input Name {type=string, default=None}
 Label Text {type=string, default=Label}
]
widgets {
 info {namedSlotCount=1, rootWidgetClass=&VerticalBox, widgetCount=8}
 widgets [
  VerticalBox_1 {widgetClassPath=&VerticalBox, widget=VerticalBox_1, properties {visibility=Visible}}
  LabelTextBlock {
   widgetClassPath=&TextBlock
   bIsVariable=true
   parent=VerticalBox_1
   widget=LabelTextBlock
   properties {
    justification=Center
    text=Label
    colorAndOpacity {colorUseRule=UseColor_Specified, specifiedColor=<rgba 0.5271 0.5271 0.5271 1>}
    font {
     bForceMonospaced=false
     bMaterialIsStencil=false
     fontMaterial=null
     fontObject=&Roboto
     letterSpacing=0
     monospacedWidth=1
     size=9
     skewAmount=0
     typefaceFontName=Regular
     outlineSettings {
      bApplyOutlineToDropShadows=false
      bMiteredCorners=false
      bSeparateFillAlpha=false
      outlineMaterial=null
      outlineSize=0
      outlineColor=<rgba 0 0 0 1>
     }
    }
   }
   slot {class=VerticalBoxSlot, properties {horizontalAlignment=HAlign_Center, padding=<ltrb 0 8 0 0>}}
  }
  Spacer {widgetClassPath=&Spacer, parent=VerticalBox_1, widget=Spacer, properties {size=<1 4>}, slot {class=VerticalBoxSlot}}
  Bool_ToggleButton {
   widgetClassPath=/TechAudioToolsContent/Widgets/MetaSoundLiterals/Bool/Primitives/WBP_MetaSoundLiteralBoolPrimitive_ToggleButton
   bIsVariable=true
   parent=VerticalBox_1
   widget=Bool_ToggleButton
   slot {class=VerticalBoxSlot, properties {horizontalAlignment=HAlign_Center}}
  }
  Content {widgetClassPath=&NamedSlot, namedSlotHost=Bool_ToggleButton, widget=Content}
  Spacer_232 {widgetClassPath=&Spacer, parent=VerticalBox_1, widget=Spacer_232, properties {size=<1 4>}, slot {class=VerticalBoxSlot}}
  Button {
   widgetClassPath=&EditorUtilityButton
   parent=VerticalBox_1
   widget=Button
   properties {
    toolTipText=Reset to default value.
    widgetStyle {
     clickedSlateSound {resourceObject=null}
     disabled {
      bIsDynamicallyLoaded=false
      drawAs=NoDrawType
      imageType=NoImage
      mirroring=NoMirror
      resourceName=None
      resourceObject=null
      tiling=NoTile
      imageSize=<32 32>
      margin=<ltrb 0 0 0 0>
      outlineSettings {
       bUseBrushTransparency=false
       roundingType=FixedRadius
       width=1
       color {colorUseRule=UseColor_ColorTable, specifiedColor=<rgba 0 0 0 0>}
       cornerRadii=<4 4 4 4>
      }
      tintColor {colorUseRule=UseColor_Specified, specifiedColor=<rgba 0.0312 0.0312 0.0312 1>}
      uVRegion {bIsValid=false, max=<0 0>, min=<0 0>}
     }
     disabledForeground {colorUseRule=UseColor_ColorTable, specifiedColor=<rgba 0 0 0 0>}
     hovered {
      bIsDynamicallyLoaded=false
      drawAs=RoundedBox
      imageType=NoImage
      mirroring=NoMirror
      resourceName=None
      resourceObject=null
      tiling=NoTile
      imageSize=<32 32>
      margin=<ltrb 0 0 0 0>
      outlineSettings {
       bUseBrushTransparency=false
       roundingType=FixedRadius
       width=0
       color {colorUseRule=UseColor_Foreground, specifiedColor=<rgba 0 0 0 0>}
       cornerRadii=<4 4 4 4>
      }
      tintColor {colorUseRule=UseColor_Specified, specifiedColor=<rgba 0.0395 0.0395 0.0395 1>}
      uVRegion {bIsValid=false, max=<0 0>, min=<0 0>}
     }
     hoveredForeground {colorUseRule=UseColor_ColorTable, specifiedColor=<rgba 0 0 0 0>}
     hoveredSlateSound {resourceObject=null}
     normal {
      bIsDynamicallyLoaded=false
      drawAs=NoDrawType
      imageType=NoImage
      mirroring=NoMirror
      resourceName=None
      resourceObject=null
      tiling=NoTile
      imageSize=<32 32>
      margin=<ltrb 0 0 0 0>
      outlineSettings {
       bUseBrushTransparency=false
       roundingType=FixedRadius
       width=1
       color {colorUseRule=UseColor_ColorTable, specifiedColor=<rgba 0 0 0 0>}
       cornerRadii=<4 4 4 4>
      }
      tintColor {colorUseRule=UseColor_Specified, specifiedColor=<rgba 0.0312 0.0312 0.0312 1>}
      uVRegion {bIsValid=false, max=<0 0>, min=<0 0>}
     }
     normalForeground {colorUseRule=UseColor_ColorTable, specifiedColor=<rgba 0 0 0 0>}
     normalPadding=<ltrb 2 0 2 0>
     pressed {
      bIsDynamicallyLoaded=false
      drawAs=RoundedBox
      imageType=NoImage
      mirroring=NoMirror
      resourceName=None
      resourceObject=null
      tiling=NoTile
      imageSize=<32 32>
      margin=<ltrb 0 0 0 0>
      outlineSettings {
       bUseBrushTransparency=false
       roundingType=FixedRadius
       width=0
       color {colorUseRule=UseColor_Foreground, specifiedColor=<rgba 0 0 0 0>}
       cornerRadii=<4 4 4 4>
      }
      tintColor {colorUseRule=UseColor_Specified, specifiedColor=<rgba 0.0395 0.0395 0.0395 1>}
      uVRegion {bIsValid=false, max=<0 0>, min=<0 0>}
     }
     pressedForeground {colorUseRule=UseColor_ColorTable, specifiedColor=<rgba 0 0 0 0>}
     pressedPadding=<ltrb 2 0.5 2 -0.5>
     pressedSlateSound {resourceObject=null}
    }
   }
   slot {class=VerticalBoxSlot, properties {horizontalAlignment=HAlign_Center, verticalAlignment=VAlign_Center}}
  }
  Icon {
   widgetClassPath=/Script/TechAudioToolsContentEditor.TechAudioToolsEditorSlateIcon
   bIsVariable=true
   parent=Button
   widget=Icon
   properties {fullIconPath=DefaultRevisionControlStyle.RevisionControl.Actions.Revert, imageColorAndOpacity=<rgba 0.2135 0.2135 0.2135 1>, size=<12 12>}
   slot {class=ButtonSlot, properties {padding=<ltrb 2 2 2 2>}}
  }
 ]
}
